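# Physics asset for the Mechanical Engineer Female pickaxe, giving its dangling parts collision shapes and joints for secondary motion.
# Capsule bodies cover the root, brace, wrench and two straps, with each dynamic part constrained directly to b_AR_Root in a star layout.
# The wrench is rotationally locked, the brace locks swing but leaves twist, and the straps, on tiny capsules, keep both swing and twist unlocked.
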PhysicsAsset /BRCosmetics/Weapons/FORT_Melee/Pickaxe_Mechanical_Engineer_Female/Meshes/Mechanical_Engineer_Female_Physics
bodies {
 b_AR_Root [{shapeType=Capsule, length=50.112, radius=8.0909, center=<6.5083 -0.0864 27.0629>, rotation=<rot 0 0 -0.1823>}]
 dyn_brace [{shapeType=Capsule, length=15.861, radius=7.9994, center=<0 6.1744 -7.8714>, rotation=<rot 0 0 141.3926>}]
 dyn_strap_btm [{shapeType=Capsule, length=2, radius=1, center=<1 0 0>, rotation=<rot 90 0 0>}]
 dyn_strap_top [{shapeType=Capsule, length=2, radius=1, center=<1 0 0>, rotation=<rot 90 0 0>}]
 dyn_wrench [{shapeType=Capsule, length=35.8345, radius=10.5645, center=<0.7866 0 0>, rotation=<rot 90 0 0>}]
}
constraints [
 {bone1Name=dyn_strap_btm, bone2Name=b_AR_Root, swing1LimitDegrees=30, swing2LimitDegrees=10, twistLimitDegrees=5}
 {bone1Name=dyn_strap_top, bone2Name=b_AR_Root, swing1LimitDegrees=30, swing2LimitDegrees=10, twistLimitDegrees=5}
 {
  bone1Name=dyn_wrench
  bone2Name=b_AR_Root
  swing1LimitDegrees=45
  swing1Motion=Locked
  swing2LimitDegrees=45
  swing2Motion=Locked
  twistLimitDegrees=45
  twistMotion=Locked
 }
 {
  bone1Name=dyn_brace
  bone2Name=b_AR_Root
  swing1LimitDegrees=0
  swing1Motion=Locked
  swing2LimitDegrees=0
  swing2Motion=Locked
  twistLimitDegrees=20
 }
]
