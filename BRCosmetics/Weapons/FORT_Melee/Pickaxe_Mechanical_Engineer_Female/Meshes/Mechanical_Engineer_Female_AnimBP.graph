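# Animation Blueprint for the female Mechanical Engineer pickaxe, adding procedural gear rotation and physics-driven secondary motion.
# Each update scales dyn_wrench's vertical offset from its reference transform, relative to b_AR_Root, into a roll additively applied to dyn_gear.
# The pose then feeds a RigidBody simulation in b_AR_Root space, using an override physics asset.
# The dyn_brace AnimDynamics node has zero alpha, and the dyn_gear one is unconnected.
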AnimBlueprint /BRCosmetics/Weapons/FORT_Melee/Pickaxe_Mechanical_Engineer_Female/Meshes/Mechanical_Engineer_Female_AnimBP
target_skeleton=&/Mechanical_Engineer_Female_Skeleton
graphs [
 AnimGraph {nodes [
   Anim Dynamics - Bone: dyn_gear(Alpha=1, ExternalForce="0,0,0") {
    bGravityOverrideInSimSpace=true
    bUseGravityOverride=true
    alphaBoolBlend {blendOption=Linear}
    boundBone {boneName=dyn_gear}
    componentAppliedLinearAccClamp=<10000 10000 10000>
    componentLinearAccScale=<0.75 0.75 0.75>
    componentLinearVelScale=<0.75 0.75 0.75>
    gravityOverride {z=-980}
    physicsBodyDefinitions [{
      sphereCollisionRadius=0
      boundBone {boneName=dyn_gear}
      boxExtents=<2.5 5 5>
      constraintSetup {twistAxis=AxisZ, angularLimitsMax=<45 0 0>, angularLimitsMin=<-45 0 0>, angularTarget=<0 0 0>}
      localJointOffset=<0 0 -0.015>
     }]
   }
   n1 = Input Pose()
   n2 = Anim Dynamics - Bone: dyn_brace(
    ComponentPose="Transform (Modify) Bone - Bone: dyn_gear"(ComponentPose=Local To Component(LocalPose=@n1), Alpha=1, Translation="0,0,0", Rotation=Get gear_rot() -> (gear_rot:Rotator), Scale="1,1,1") {rotationMode=BMM_Additive, alphaBoolBlend {blendOption=Linear}, boneToModify {boneName=dyn_gear}}
    Alpha=0
    ExternalForce="100,0,0"
   ) {
    angularSpringConstant=10
    bGravityOverrideInSimSpace=true
    bOverrideAngularDamping=true
    bUseGravityOverride=true
    alphaBoolBlend {blendOption=Linear}
    boundBone {boneName=dyn_brace}
    componentAppliedLinearAccClamp=<10000 10000 10000>
    componentLinearAccScale=<0.75 0.75 0.75>
    componentLinearVelScale=<0.75 0.75 0.75>
    gravityOverride {z=-980}
    physicsBodyDefinitions [{
      sphereCollisionRadius=0
      boundBone {boneName=dyn_brace}
      boxExtents=<5 10 10>
      constraintSetup {twistAxis=AxisZ, angularLimitsMax=<30 0 0>, angularLimitsMin=<-30 0 0>, angularTarget=<0 0 0>}
      localJointOffset=<0 1 0>
     }]
   }
   n3 = RigidBody(ComponentPose=@n2) {
    bClampLinearTranslationLimitToRefPose=true
    bOverrideWorldGravity=true
    overridePhysicsAsset=&/Mechanical_Engineer_Female_Physics
    simulationSpace=BaseBoneSpace
    alphaBoolBlend {blendOption=Linear}
    baseBoneRef {boneName=b_AR_Root}
    componentLinearAccScale=<0.6 0.3 0.3>
    componentLinearVelScale=<0.6 0.3 0.3>
    overrideWorldGravity {z=-980}
   }
   Output Pose(Result=Component To Local(ComponentPose=@n3)) {layerGroup=None}
  ]}
 EventGraph {nodes [
   Event BlueprintUpdateAnimation()
   n1 = GetAlphaFromReferenceTransformDelta(
    SkeletalMeshComponent=Get skel_mesh() -> (skel_mesh:Skeletal Mesh Component Object Reference)
    BoneName=dyn_wrench
    BaseName=b_AR_Root
    DeltaComponentType=TranslationZ
    InputRange=(X=-1,Y=1)
    OutputRange=(X=-1,Y=1)
    TwistAxis="1,0,0"
   )
   > Set gear_rot(gear_rot:Rotator=MakeRotator(Roll=float * float(A=@n1.RawOutput, B=20))) -> (Output_Get:Rotator)
   Event BlueprintInitializeAnimation()
   > Set skel_mesh(skel_mesh:Skeletal Mesh Component Object Reference=GetPartSkeletalMeshComponent()) -> (Output_Get:Skeletal Mesh Component Object Reference)
   TryGetPawnOwner()
  ]}
]
variables [
 gear_rot {type=Rotator}
 skel_mesh {type=SkeletalMeshComponent}
]
